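# Additive unlit material for Niagara mesh particles that projects graffiti decals for the GraffitiTon pickaxe effects.
# Scene position behind the translucent mesh, taken relative to the particle in instance space and divided by a dynamic box size, yields decal UVs and a depth fade.
# The UVs sample a quarter-width slice of a grayscale texture, offset by a dynamic flipbook parameter.
# Dynamic erosion is subtracted and smoothstepped into a mask, alongside a two-color gradient.
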Material /BRCosmetics/Weapons/FORT_Melee/Pickaxe_GraffitiTon/FX/M_Pickaxe_GraffitiTon_Decal
graph {
 properties {
  bAutomaticallySetUsageInEditor=false
  bUsedWithInstancedStaticMeshes=true
  bUsedWithNiagaraMeshParticles=true
  blendMode=BLEND_Additive
  shadingModel=MSM_Unlit
  usedShadingModels=Unlit
 }
 nodes [
  Constant3Vector() {constant=<rgba 1 0.2955 0.724 1>}
  n1 = DynamicParameter() {defaultValue=<rgba 1 0 1 1>, paramNames [DecalProjectionBoxSize, TextureErosion, FlipbookSelect, Param4]}
  n2 = Add(A=Multiply(A=NamedRerouteUsage(), B=Constant2Vector() {g=1, r=0.25}), B=AppendVector(A=@n1.FlipbookSelect, B=Constant()))
  n3 = TextureSampleParameter2D(UVs=@n2) {
   parameterName=MainTexture
   samplerType=SAMPLERTYPE_LinearGrayscale
   texture=&/T_Pickaxe_GraffitiTon_Deca
  }
  n4 = Subtract(A=@n3.R, B=@n1.TextureErosion)
  n5 = Saturate(None=SmoothStep(Value=@n4) {constMax=0.8, constMin=0.02})
  LinearInterpolate(
   A=Constant() {r=1}
   B=LinearInterpolate(
    A=VectorParameter() {parameterName=ColorB, defaultValue=<rgba 1.8122 0.895 2 1>}.RGB
    B=VectorParameter() {parameterName=ColorA, defaultValue=<rgba 2.5 0.9177 2.1258 1>}.RGB
    Alpha=@n4
   )
   Alpha=@n5
  )
  MaterialFunctionCall(Emissive=ParticleColor().RGB) {materialFunction=&MF_FX_EmissiveAdjustments}
  n6 = Subtract(
   A=MaterialFunctionCall() {materialFunction=&WorldPositionBehindTranslucency}.Result
   B=ParticlePositionWS()
  )
  n7 = Add(
   A=Divide(A=Transform(None=@n6) {transformSourceType=TRANSFORMSOURCE_World, transformType=TRANSFORM_Instance}, B=@n1.DecalProjectionBoxSize) {constB=1000}
   B=ScalarParameter() {defaultValue=0.5, parameterName=Offset}
  ) {constB=0.5}
  n8 = ComponentMask(None=@n7) {b=true}
  n9 = Sine(None=Multiply(A=Saturate(None=@n8)) {constB=0.5})
  Multiply(A=@n5, B=Multiply(A=@n3.R, B=Saturate(None=@n9)))
  n10 = ComponentMask(None=@n7) {g=true, r=true}
  n11 = Constant2Vector() {g=1, r=-1}
  NamedRerouteDeclaration(None=Saturate(None=MaterialFunctionCall(UVs=@n10, Texture Scale=@n11) {materialFunction=&ScaleUVsByCenter}.UVs)) {name=DecalUV, nodeColor=<rgba 0 0.8125 1 1>}
  n12 = ComponentMask(None=Abs(None=SceneTexture() {sceneTextureId=PPI_WorldNormal}.Color)) {g=true, r=true}
  n13 = Constant2Vector() {g=1}
  n14 = Add(A=@n2, B=Multiply(A=Multiply(A=@n12, B=@n13)))
  QualitySwitch(Default=@n14, Low=@n2, High=@n14, Medium=@n2, Epic=@n14)
  ScalarParameter() {defaultValue=100, parameterName=BoxSize}
  ScalarParameter() {defaultValue=1, parameterName=UVScale}
 ]
}
